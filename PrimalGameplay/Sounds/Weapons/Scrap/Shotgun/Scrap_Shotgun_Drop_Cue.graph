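# Sound cue for the drop sound of the Scrap Shotgun weapon.
# A Modulator at default settings varies pitch and volume on a single wave player for Scrap_Shotgun_Drop_01.
# It plays at reduced volume through shared weapon-drop attenuation and the loot-collect sound class.
# It joins the Weapon_Drop_StopOldest concurrency set; the stored PreventNew override values have no override flag shown enabled.
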SoundCue /PrimalGameplay/Sounds/Weapons/Scrap/Shotgun/Scrap_Shotgun_Drop_Cue
root=Modulator_1
nodes [
 {id=Modulator_1, children [WavePlayer_0]}
 {id=WavePlayer_0, properties {soundWaveAssetPtr=&/Scrap_Shotgun_Drop_01}}
]
properties {
 attenuationSettings=&Fort_Weapon_Drop_Att
 maxDistance=3000
 soundClassObject=&Fort_Loot_Collect
 volumeMultiplier=0.7
 concurrencyOverrides {resolutionRule=PreventNew}
 concurrencySet [&Weapon_Drop_StopOldest]
}
