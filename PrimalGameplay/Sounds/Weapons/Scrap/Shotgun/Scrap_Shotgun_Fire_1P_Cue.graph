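# Sound cue for the first-person firing sound of the Scrap Shotgun.
# A Random node picks evenly among four close-range fire waves, then a Modulator adds slight upward pitch and volume variation.
# It uses first-person weapon attenuation, a ranged-fire 1P sound class and a first-person gunfire concurrency set.
# Concurrency override values specify a small per-owner voice cap that stops the oldest.
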SoundCue /PrimalGameplay/Sounds/Weapons/Scrap/Shotgun/Scrap_Shotgun_Fire_1P_Cue
root=Modulator_1
nodes [
 {id=Modulator_1, children [Random_2], properties {pitchMax=1.1, pitchMin=1, volumeMax=1.1, volumeMin=1}}
 {id=Random_2, children [WavePlayer_2,WavePlayer_3,WavePlayer_4,WavePlayer_5], properties {weights [1,1,1,1]}}
 {id=WavePlayer_2, properties {soundWaveAssetPtr=&/Scrap_Shotgun_Fire_Close_04}}
 {id=WavePlayer_3, properties {soundWaveAssetPtr=&/Scrap_Shotgun_Fire_Close_01}}
 {id=WavePlayer_4, properties {soundWaveAssetPtr=&/Scrap_Shotgun_Fire_Close_02}}
 {id=WavePlayer_5, properties {soundWaveAssetPtr=&/Scrap_Shotgun_Fire_Close_03}}
]
properties {
 attenuationSettings=&Fort_Weapon_Projectile_FirstPerson_Athena_Att
 maxDistance=2097152
 soundClassObject=&Fort_WeaponRanged_Fire_1P_PSM
 volumeMultiplier=1.1
 attenuationOverrides {falloffDistance=5000, attenuationShapeExtents {x=1000}}
 concurrencyOverrides {bLimitToOwner=true, maxCount=3, resolutionRule=StopOldest}
 concurrencySet [&Weapon_Gunfire_FirstPerson_Conc]
}
