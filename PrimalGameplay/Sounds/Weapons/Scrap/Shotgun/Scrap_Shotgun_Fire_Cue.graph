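# Sound cue for Scrap Shotgun gunfire, layering close, medium and distant shots, classed as third-person ranged weapon fire.
# A mixer blends the three layers at descending levels, each with its own sound class override and distance attenuation asset.
# Each layer picks evenly among recorded variants through a random node, then a modulator varies volume, plus pitch on the close layer.
# Local and global gunfire concurrency sets govern overlapping shots.
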SoundCue /PrimalGameplay/Sounds/Weapons/Scrap/Shotgun/Scrap_Shotgun_Fire_Cue
root=Mixer_0
nodes [
 {id=Attenuation_0, children [Modulator_3], properties {attenuationSettings=&Fort_Weapon_Projectile_Close_Athena_Att}}
 {
  id=Attenuation_1
  children [Modulator_4]
  properties {attenuationSettings=&Fort_Weapon_Projectile_Medium_Mono_Athena_Att}
 }
 {
  id=Attenuation_2
  children [Modulator_5]
  properties {attenuationSettings=&Fort_Weapon_Projectile_Distant_Mono_Athena_Att}
 }
 {id=Mixer_0, children [SoundClass_0,SoundClass_6,SoundClass_1], properties {inputVolume [1,0.6,0.3]}}
 {id=Modulator_3, children [Random_2], properties {pitchMax=1.1, pitchMin=1, volumeMax=1.1, volumeMin=1}}
 {id=Modulator_4, children [Random_0], properties {volumeMax=2, volumeMin=1.8}}
 {id=Modulator_5, children [Random_1], properties {volumeMax=1.7, volumeMin=1.6}}
 {id=Random_0, children [WavePlayer_0,WavePlayer_1,WavePlayer_6,WavePlayer_7], properties {weights [1,1,1,1]}}
 {id=Random_1, children [WavePlayer_8, WavePlayer_9, WavePlayer_10], properties {weights [1,1,1]}}
 {id=Random_2, children [WavePlayer_2,WavePlayer_3,WavePlayer_4,WavePlayer_5], properties {weights [1,1,1,1]}}
 {
  id=SoundClass_0
  children [Attenuation_0]
  properties {soundClassOverride=&Fort_WeaponRanged_Fire_3P_PSM}
 }
 {
  id=SoundClass_1
  children [Attenuation_2]
  properties {soundClassOverride=&Fort_WeaponRanged_Fire_Distant}
 }
 {
  id=SoundClass_6
  children [Attenuation_1]
  properties {soundClassOverride=&Fort_WeaponRanged_Fire_3P_PSM}
 }
 {id=WavePlayer_0, properties {soundWaveAssetPtr=&/Scrap_Shotgun_Fire_Medium_01}}
 {id=WavePlayer_1, properties {soundWaveAssetPtr=&/Scrap_Shotgun_Fire_Medium_02}}
 {id=WavePlayer_10, properties {soundWaveAssetPtr=&/Scrap_Shotgun_Fire_Distant_03}}
 {id=WavePlayer_2, properties {soundWaveAssetPtr=&/Scrap_Shotgun_Fire_Close_04}}
 {id=WavePlayer_3, properties {soundWaveAssetPtr=&/Scrap_Shotgun_Fire_Close_01}}
 {id=WavePlayer_4, properties {soundWaveAssetPtr=&/Scrap_Shotgun_Fire_Close_02}}
 {id=WavePlayer_5, properties {soundWaveAssetPtr=&/Scrap_Shotgun_Fire_Close_03}}
 {id=WavePlayer_6, properties {soundWaveAssetPtr=&/Scrap_Shotgun_Fire_Medium_03}}
 {id=WavePlayer_7, properties {soundWaveAssetPtr=&/Scrap_Shotgun_Fire_Medium_04}}
 {id=WavePlayer_8, properties {soundWaveAssetPtr=&/Scrap_Shotgun_Fire_Distant_01}}
 {id=WavePlayer_9, properties {soundWaveAssetPtr=&/Scrap_Shotgun_Fire_Distant_02}}
]
properties {
 maxDistance=27500
 soundClassObject=&Fort_WeaponRanged_Fire_3P_PSM
 volumeMultiplier=0.9
 attenuationOverrides {falloffDistance=5000, attenuationShapeExtents {x=1000}}
 concurrencyOverrides {bLimitToOwner=true, maxCount=3, resolutionRule=StopOldest}
 concurrencySet [&Weapon_Gunfire_Local_Conc, &Weapon_Gunfire_Global_Conc]
}
