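# Sound cue for picking up the Scrap Shotgun weapon.
# A single wave player feeds a modulator at the root, using pickup attenuation and the loot collect sound class at reduced volume.
# Overridden concurrency allows one instance at a time and prevents new plays while it is active.
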SoundCue /PrimalGameplay/Sounds/Weapons/Scrap/Shotgun/Scrap_Shotgun_Pickup_Cue
root=Modulator_0
nodes [
 {id=Modulator_0, children [WavePlayer_0]}
 {id=WavePlayer_0, properties {soundWaveAssetPtr=&/Scrap_Shotgun_Pickup_01}}
]
properties {
 attenuationSettings=&Fort_Weapon_Pickup_Att
 bOverrideConcurrency=true
 maxDistance=3500
 soundClassObject=&Fort_Loot_Collect
 volumeMultiplier=0.2769
 concurrencyOverrides {maxCount=1, resolutionRule=PreventNew}
}
