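# Sound cue for the end of the scrap shotgun's reload, played as weapon foley.
# A random node picks one of two equally weighted recordings, then a fixed volume boost and a short fixed delay apply.
# It uses the shared weapon reload attenuation and overrides concurrency to cap simultaneous plays per owner, stopping the oldest.
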SoundCue /PrimalGameplay/Sounds/Weapons/Scrap/Shotgun/Scrap_Shotgun_ReloadEnd_Cue
root=Delay_0
nodes [
 {id=Delay_0, children [Modulator_0], properties {delayMax=0.07, delayMin=0.07}}
 {id=Modulator_0, children [Random_0], properties {volumeMax=1.9, volumeMin=1.9}}
 {id=Random_0, children [WavePlayer_0,WavePlayer_2], properties {weights [1,1]}}
 {id=WavePlayer_0, properties {soundWaveAssetPtr=&/Scrap_Shotgun_ReloadEnd_01}}
 {id=WavePlayer_2, properties {soundWaveAssetPtr=&/Scrap_Shotgun_ReloadEnd_02}}
]
properties {
 attenuationSettings=&Fort_Weapon_Reload_Att
 bOverrideConcurrency=true
 maxDistance=2250
 soundClassObject=&Fort_WeaponFoley
 volumeMultiplier=0.401
 attenuationOverrides {falloffDistance=1600}
 concurrencyOverrides {bLimitToOwner=true, maxCount=2, resolutionRule=StopOldest}
}
