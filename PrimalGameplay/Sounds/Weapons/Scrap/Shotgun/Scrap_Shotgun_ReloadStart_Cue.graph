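# Sound cue for the start of the Scrap shotgun reload, played as weapon foley.
# A random node picks one of three equally weighted reload-start wave variants and passes it through a modulator at the root.
# It uses the shared weapon reload attenuation and overrides concurrency so new plays are blocked once the limit is reached.
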SoundCue /PrimalGameplay/Sounds/Weapons/Scrap/Shotgun/Scrap_Shotgun_ReloadStart_Cue
root=Modulator_1
nodes [
 {id=Modulator_1, children [Random_0]}
 {id=Random_0, children [WavePlayer_1,WavePlayer_2,WavePlayer_3], properties {weights [1,1,1]}}
 {id=WavePlayer_1, properties {soundWaveAssetPtr=&/Scrap_Shotgun_ReloadStart_01}}
 {id=WavePlayer_2, properties {soundWaveAssetPtr=&/Scrap_Shotgun_ReloadStart_02}}
 {id=WavePlayer_3, properties {soundWaveAssetPtr=&/Scrap_Shotgun_ReloadStart_03}}
]
properties {
 attenuationSettings=&Fort_Weapon_Reload_Att
 bOverrideConcurrency=true
 maxDistance=2250
 soundClassObject=&Fort_WeaponFoley
 volumeMultiplier=0.551
 attenuationOverrides {falloffDistance=1600}
 concurrencyOverrides {resolutionRule=PreventNew}
}
